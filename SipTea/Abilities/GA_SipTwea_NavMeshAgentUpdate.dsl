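; Gameplay ability that changes an AI bot's default navigation filter class when activated, added in S29 to stop Stormtroopers destroying or climbing cheap obstacles while engaging Chewie.
; On activation it gets the avatar's Fort player pawn, casts that pawn's controller to the Athena AI bot controller and applies the NewNavigationClass variable if valid.
; It ends right after the successful cast path, while failed pawn, controller or cast checks leave it without an EndAbility call.
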(event Ability|EventActivateAbility
  (bind _newnavigationclass (Variables|Default|GetNewNavigationClass))
  ;[1 1 1] This Ability was added in S29 to make sure Stormtroopers didn't try to destroy / climb on Cheap Obstacles while engaging chewie. Can be removed if we re-use them again.
  (bind _outfortplayerpawn (CallFunction|GetAvatarasValidFortPlayerPawn)
    (:ValidFortPlayerPawn
      (Utilities|IsValid (Pawn|GetController _outfortplayerpawn)
        (:Is%20Valid
          (bind _asfort_athena_aibot_controller (Utilities|Casting|CastToFortAthenaAIBotController (Pawn|GetController _outfortplayerpawn))
            (:then
              (if (Utilities|IsValidClass _newnavigationclass)
                (Variables|/Script/AIModule.AIController|SetDefaultNavigationFilterClass _newnavigationclass _asfort_athena_aibot_controller)
                (else
                  _))
              (Ability|EndAbility))
            (:CastFailed
              _)))
        (:Is%20Not%20Valid
          _)))
    (:AvatarCastFailed
      _)))

(event Ability|EventFailedToActivatePassiveAbility
  (params (ActorInfo :type Gameplay%20Ability%20Actor%20Info%20Structure)))
